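# Level Sequence for the female PearlTaffyOpera skin's lobby idle animation, built at 30 fps over frames 0–470.
# The main mannequin at the origin carries body and on-face Control Rig tracks, plus a bone-tweaker rig on its head.
# Face keys drive eye-aim gaze shifts and asymmetric brow, lid and mouth expressions that return to the starting pose.
# Back, Left and Right mannequins stand at fixed transforms around it, each playing a skeletal animation.
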LevelSequence /BRCosmetics/Animation/Game/MainPlayer/Menu/BR/CustomSkins/_S41/PearlTaffyOpera_F/PearlTaffyOpera_Female_LobbyIdle_LS
bindings [
 Back {
  id=#1
  possessed_class=FortPlayerMannequin
  tracks [
   {type=MovieSceneSpawnTrack, sections [{type=MovieSceneSpawnSection, channels [None {type=MovieSceneScriptingBoolChannel, default=true}]}]}
   {
    type=MovieScene3DTransformTrack
    sections [{
      type=MovieScene3DTransformSection
      channels [
       Location.X {type=MovieSceneScriptingDoubleChannel, default=229.47}
       Location.Y {type=MovieSceneScriptingDoubleChannel, default=-189.38}
       Location.Z {type=MovieSceneScriptingDoubleChannel, default=0}
       Rotation.X {type=MovieSceneScriptingDoubleChannel, default=0}
       Rotation.Y {type=MovieSceneScriptingDoubleChannel, default=0}
       Rotation.Z {type=MovieSceneScriptingDoubleChannel, default=-145.97}
       Scale.X {type=MovieSceneScriptingDoubleChannel, default=1}
       Scale.Y {type=MovieSceneScriptingDoubleChannel, default=1}
       Scale.Z {type=MovieSceneScriptingDoubleChannel, default=1}
      ]
     }]
   }
   {type=MovieSceneSkeletalAnimationTrack, sections [{type=MovieSceneSkeletalAnimationSection, end_frame=470, start_frame=0}]}
  ]
 }
 SkeletalMeshComponent0 {id=#2, parent_id=#3, possessed_class=SkeletalMeshComponentBudgeted, tracks []}
 Left {
  id=#4
  possessed_class=FortPlayerMannequin
  tracks [
   {type=MovieSceneSpawnTrack, sections [{type=MovieSceneSpawnSection, channels [None {type=MovieSceneScriptingBoolChannel, default=true}]}]}
   {
    type=MovieScene3DTransformTrack
    sections [{
      type=MovieScene3DTransformSection
      channels [
       Location.X {type=MovieSceneScriptingDoubleChannel, default=-122.32}
       Location.Y {type=MovieSceneScriptingDoubleChannel, default=-76.68}
       Location.Z {type=MovieSceneScriptingDoubleChannel, default=0}
       Rotation.X {type=MovieSceneScriptingDoubleChannel, default=0}
       Rotation.Y {type=MovieSceneScriptingDoubleChannel, default=0}
       Rotation.Z {type=MovieSceneScriptingDoubleChannel, default=96.76}
       Scale.X {type=MovieSceneScriptingDoubleChannel, default=1}
       Scale.Y {type=MovieSceneScriptingDoubleChannel, default=1}
       Scale.Z {type=MovieSceneScriptingDoubleChannel, default=1}
      ]
     }]
   }
   {type=MovieSceneSkeletalAnimationTrack, sections [{type=MovieSceneSkeletalAnimationSection, end_frame=470, start_frame=0}]}
  ]
 }
 Right {
  id=#3
  possessed_class=FortPlayerMannequin
  tracks [
   {type=MovieSceneSpawnTrack, sections [{type=MovieSceneSpawnSection, channels [None {type=MovieSceneScriptingBoolChannel, default=true}]}]}
   {
    type=MovieScene3DTransformTrack
    sections [{
      type=MovieScene3DTransformSection
      channels [
       Location.X {type=MovieSceneScriptingDoubleChannel, default=119.61}
       Location.Y {type=MovieSceneScriptingDoubleChannel, default=-74.97}
       Location.Z {type=MovieSceneScriptingDoubleChannel, default=0}
       Rotation.X {type=MovieSceneScriptingDoubleChannel, default=0}
       Rotation.Y {type=MovieSceneScriptingDoubleChannel, default=0}
       Rotation.Z {type=MovieSceneScriptingDoubleChannel, default=-57.17}
       Scale.X {type=MovieSceneScriptingDoubleChannel, default=1}
       Scale.Y {type=MovieSceneScriptingDoubleChannel, default=1}
       Scale.Z {type=MovieSceneScriptingDoubleChannel, default=1}
      ]
     }]
   }
   {type=MovieSceneSkeletalAnimationTrack, sections [{type=MovieSceneSkeletalAnimationSection, end_frame=470, start_frame=0}]}
  ]
 }
 FortPlayerMannequin {
  id=#5
  possessed_class=FortPlayerMannequin
  tracks [
   {type=MovieSceneSpawnTrack, sections [{type=MovieSceneSpawnSection, channels [None {type=MovieSceneScriptingBoolChannel, default=true}]}]}
   {type=MovieSceneControlRigParameterTrack, display_name=MR_Biped_v02, sections [{type=MovieSceneControlRigParameterSection}]}
   {
    type=MovieSceneControlRigParameterTrack
    display_name=Fortnite_3L_OnFace_CtrlRig
    sections [{
      type=MovieSceneControlRigParameterSection
      channels [
       CTRL_C_eyesAim.Location.X {
        type=MovieSceneScriptingFloatChannel
        default=-17.3153
        thinned_from=14
        keys [[0,-17.6805],[56,-17.6652],[77,-36.4239],[113,-17.6805],[136,-17.6805],[177,-81],[259,-81],[296,-17.6823],[402,-17.6805],[428,-169.5],[465,-17.6805],[470,-17.6805]]
       }
       CTRL_C_eyesAim.Location.Y {
        type=MovieSceneScriptingFloatChannel
        default=-634.4464
        thinned_from=14
        keys [[0,-634.49],[56,-171.7264],[77,-237.3089],[113,-634.49],[470,-634.49]]
       }
       CTRL_C_eyesAim.Location.Z {
        type=MovieSceneScriptingFloatChannel
        default=21.6759
        thinned_from=14
        keys [[0,51.2733],[56,-195.5],[77,-195.5104],[113,49.5],[136,49.5],[177,-41],[259,-41],[296,645],[366,1073],[402,51.2733],[470,51.2733]]
       }
       CTRL_C_eyesAim.Rotation.X {type=MovieSceneScriptingFloatChannel, default=0, const [0,14]}
       CTRL_C_eyesAim.Rotation.Y {type=MovieSceneScriptingFloatChannel, default=0, const [0,14]}
       CTRL_C_eyesAim.Rotation.Z {type=MovieSceneScriptingFloatChannel, default=0, const [0,14]}
       CTRL_C_eyesAim.Scale.X {type=MovieSceneScriptingFloatChannel, default=1, const [1,14]}
       CTRL_C_eyesAim.Scale.Y {type=MovieSceneScriptingFloatChannel, default=1, const [1,14]}
       CTRL_C_eyesAim.Scale.Z {type=MovieSceneScriptingFloatChannel, default=1, const [1,14]}
       CTRL_L_eyeAim.Location.X {type=MovieSceneScriptingFloatChannel, default=0, const [0,14]}
       CTRL_L_eyeAim.Location.Y {type=MovieSceneScriptingFloatChannel, default=0, const [0,14]}
       CTRL_L_eyeAim.Location.Z {type=MovieSceneScriptingFloatChannel, default=0, const [0,14]}
       CTRL_L_eyeAim.Rotation.X {type=MovieSceneScriptingFloatChannel, default=0, const [0,14]}
       CTRL_L_eyeAim.Rotation.Y {type=MovieSceneScriptingFloatChannel, default=0, const [0,14]}
       CTRL_L_eyeAim.Rotation.Z {type=MovieSceneScriptingFloatChannel, default=0, const [0,14]}
       CTRL_L_eyeAim.Scale.X {type=MovieSceneScriptingFloatChannel, default=1, const [1,14]}
       CTRL_L_eyeAim.Scale.Y {type=MovieSceneScriptingFloatChannel, default=1, const [1,14]}
       CTRL_L_eyeAim.Scale.Z {type=MovieSceneScriptingFloatChannel, default=1, const [1,14]}
       CTRL_R_eyeAim.Location.X {type=MovieSceneScriptingFloatChannel, default=0, const [0,14]}
       CTRL_R_eyeAim.Location.Y {type=MovieSceneScriptingFloatChannel, default=0, const [0,14]}
       CTRL_R_eyeAim.Location.Z {type=MovieSceneScriptingFloatChannel, default=0, const [0,14]}
       CTRL_R_eyeAim.Rotation.X {type=MovieSceneScriptingFloatChannel, default=0, const [0,14]}
       CTRL_R_eyeAim.Rotation.Y {type=MovieSceneScriptingFloatChannel, default=0, const [0,14]}
       CTRL_R_eyeAim.Rotation.Z {type=MovieSceneScriptingFloatChannel, default=0, const [0,14]}
       CTRL_R_eyeAim.Scale.X {type=MovieSceneScriptingFloatChannel, default=1, const [1,14]}
       CTRL_R_eyeAim.Scale.Y {type=MovieSceneScriptingFloatChannel, default=1, const [1,14]}
       CTRL_R_eyeAim.Scale.Z {type=MovieSceneScriptingFloatChannel, default=1, const [1,14]}
       CTRL_convergenceSwitch {type=MovieSceneScriptingFloatChannel, default=0, const [0,14]}
       follow_head {type=MovieSceneScriptingFloatChannel, default=0, const [0,14]}
       Face_Ctrl_Master.Location.X {type=MovieSceneScriptingFloatChannel, default=0, const [0,14]}
       Face_Ctrl_Master.Location.Y {type=MovieSceneScriptingFloatChannel, default=0, const [0,14]}
       Face_Ctrl_Master.Location.Z {type=MovieSceneScriptingFloatChannel, default=0, const [0,14]}
       Face_Ctrl_Master.Rotation.X {type=MovieSceneScriptingFloatChannel, default=0, const [0,14]}
       Face_Ctrl_Master.Rotation.Y {type=MovieSceneScriptingFloatChannel, default=0, const [0,14]}
       Face_Ctrl_Master.Rotation.Z {type=MovieSceneScriptingFloatChannel, default=0, const [0,14]}
       Face_Ctrl_Master.Scale.X {type=MovieSceneScriptingFloatChannel, default=1, const [1,14]}
       Face_Ctrl_Master.Scale.Y {type=MovieSceneScriptingFloatChannel, default=1, const [1,14]}
       Face_Ctrl_Master.Scale.Z {type=MovieSceneScriptingFloatChannel, default=1, const [1,14]}
       gui_follow_head {type=MovieSceneScriptingFloatChannel, default=1, const [1,14]}
       CTRL_L_eye_blink {type=MovieSceneScriptingFloatChannel, default=0, const [0,14]}
       CTRL_L_eye_eyelidU.X {type=MovieSceneScriptingFloatChannel, default=0, thinned_from=14, keys [[0,0],[470,0]]}
       CTRL_L_eye_eyelidU.Y {
        type=MovieSceneScriptingFloatChannel
        default=0
        thinned_from=14
        keys [[0,-0.2026],[56,0],[77,-0.1272],[113,-0.3318],[136,-0.3318],[177,-0.2026],[259,-0.2026],[296,0],[428,0],[465,-0.2026],[470,-0.2026]]
       }
       CTRL_L_eye_eyelidD.X {type=MovieSceneScriptingFloatChannel, default=0, thinned_from=14, keys [[0,0],[470,0]]}
       CTRL_L_eye_eyelidD.Y {
        type=MovieSceneScriptingFloatChannel
        default=0
        thinned_from=14
        keys [[0,0],[259,0],[296,0.4264],[366,0.2901],[402,0.2901],[428,0.1415],[465,0],[470,0]]
       }
       CTRL_L_brow.X {type=MovieSceneScriptingFloatChannel, default=0, thinned_from=14, keys [[0,0],[470,0]]}
       CTRL_L_brow.Y {
        type=MovieSceneScriptingFloatChannel
        default=-0.441
        thinned_from=14
        keys [[0,-0.441],[56,-1],[77,-1],[113,-0.441],[136,-0.441],[177,-0.4338],[259,-0.4338],[296,-0.2254],[366,-0.0293],[402,-0.0293],[428,-0.441],[470,-0.441]]
       }
       CTRL_L_brow_inner {
        type=MovieSceneScriptingFloatChannel
        default=-0.0369
        thinned_from=14
        keys [[0,0.2682],[56,0.7001],[77,0.4276],[113,0.1984],[136,0.1984],[177,0.49],[259,0.49],[296,0.4168],[366,-0.0369],[428,-0.0369],[465,0.2682],[470,0.2682]]
       }
       CTRL_L_brow_outer {type=MovieSceneScriptingFloatChannel, default=0, const [0,14]}
       CTRL_C_nose.X {type=MovieSceneScriptingFloatChannel, default=0, thinned_from=14, keys [[0,0],[470,0]]}
       CTRL_C_nose.Y {type=MovieSceneScriptingFloatChannel, default=0, thinned_from=14, keys [[0,0],[77,0],[113,-0.0907],[136,-0.0907],[177,0],[470,0]]}
       CTRL_L_nose.X {type=MovieSceneScriptingFloatChannel, default=0, const [0,14]}
       CTRL_L_nose.Y {type=MovieSceneScriptingFloatChannel, default=0, const [0,14]}
       noseWrinkleUpper {type=MovieSceneScriptingFloatChannel, default=0, const [0,14]}
       CTRL_R_nose.X {type=MovieSceneScriptingFloatChannel, default=0, thinned_from=14, keys [[0,0],[470,0]]}
       CTRL_R_nose.Y {
        type=MovieSceneScriptingFloatChannel
        default=0
        thinned_from=14
        keys [[0,-0.4653],[56,0],[366,0],[402,-0.4572],[428,-0.4613],[465,-0.4653],[470,-0.4653]]
       }
       noseWrinkleUpper_2 {type=MovieSceneScriptingFloatChannel, default=0, const [0,14]}
       CTRL_L_eye_squintInner {type=MovieSceneScriptingFloatChannel, default=0, const [0,14]}
       CTRL_L_eye_cheekRaise {type=MovieSceneScriptingFloatChannel, default=0, const [0,14]}
       CTRL_C_mouth.X {type=MovieSceneScriptingFloatChannel, default=0, thinned_from=14, keys [[0,0],[470,0]]}
       CTRL_C_mouth.Y {type=MovieSceneScriptingFloatChannel, default=0, const [0,14]}
       CTRL_L_mouth_lipU.X {type=MovieSceneScriptingFloatChannel, default=0, thinned_from=14, keys [[0,0],[470,0]]}
       CTRL_L_mouth_lipU.Y {type=MovieSceneScriptingFloatChannel, default=0, thinned_from=14, keys [[0,0],[259,0],[296,0.1954],[366,0.12],[402,0.001],[470,0]]}
       mouthPress {type=MovieSceneScriptingFloatChannel, default=0, const [0,14]}
       lipsTogether {type=MovieSceneScriptingFloatChannel, default=0, const [0,14]}
       CTRL_L_mouth_lipD.X {type=MovieSceneScriptingFloatChannel, default=0, thinned_from=14, keys [[0,0],[470,0]]}
       CTRL_L_mouth_lipD.Y {type=MovieSceneScriptingFloatChannel, default=0, thinned_from=14, keys [[0,0],[259,0],[296,-0.1282],[366,-0.0787],[402,-0.0006],[470,0]]}
       mouthPress_2 {
        type=MovieSceneScriptingFloatChannel
        default=0
        thinned_from=14
        keys [[0,0],[259,0],[296,0.1502],[366,0.1502],[402,0.0026],[428,0.0013],[465,0],[470,0]]
       }
       lipsTogether_2 {
        type=MovieSceneScriptingFloatChannel
        default=0
        thinned_from=14
        keys [[0,0],[259,0],[296,0.2349],[366,0.2349],[402,0.0041],[428,0.0021],[465,0],[470,0]]
       }
       CTRL_L_mouth_smileDepress.X {type=MovieSceneScriptingFloatChannel, default=0, thinned_from=14, keys [[0,0],[259,0],[296,0.0517],[366,0.0318],[402,0.0003],[470,0]]}
       CTRL_L_mouth_smileDepress.Y {
        type=MovieSceneScriptingFloatChannel
        default=0
        thinned_from=14
        keys [[0,0],[136,0],[177,0.1803],[259,0.1803],[296,0.1445],[366,0.0887],[402,0.0007],[470,0]]
       }
       CTRL_L_mouth_corner.X {
        type=MovieSceneScriptingFloatChannel
        default=0
        thinned_from=14
        keys [[0,0],[56,0.3306],[77,0],[259,0],[296,0.3034],[366,0.1863],[402,0.0015],[428,0.0007],[470,0]]
       }
       CTRL_L_mouth_corner.Y {
        type=MovieSceneScriptingFloatChannel
        default=0.4092
        thinned_from=14
        keys [[0,0.4092],[56,0.6613],[77,0.4092],[113,0.3589],[259,0.3589],[296,-0.1716],[366,-0.1054],[402,0.4012],[428,0.4052],[465,0.4092],[470,0.4092]]
       }
       lipsBlow {type=MovieSceneScriptingFloatChannel, default=0, const [0,14]}
       mouthDimple {type=MovieSceneScriptingFloatChannel, default=0, const [0,14]}
       mouthStretch {type=MovieSceneScriptingFloatChannel, default=0, const [0,14]}
       mouthPurseU {
        type=MovieSceneScriptingFloatChannel
        default=0
        thinned_from=14
        keys [[0,0],[259,0],[296,0.4888],[366,0.4888],[402,0.0085],[428,0.0043],[465,0],[470,0]]
       }
       mouthPurseD {type=MovieSceneScriptingFloatChannel, default=0, const [0,14]}
       mouthFunnelU {type=MovieSceneScriptingFloatChannel, default=0, thinned_from=14, keys [[0,0],[77,0],[113,0.0139],[259,0.0139],[296,0],[470,0]]}
       mouthFunnelD {type=MovieSceneScriptingFloatChannel, default=0, thinned_from=14, keys [[0,0],[77,0],[113,0.0139],[259,0.0139],[296,0],[470,0]]}
       mouthCornerSharpnessU {type=MovieSceneScriptingFloatChannel, default=0, const [0,14]}
       mouthCornerSharpnessD {type=MovieSceneScriptingFloatChannel, default=0, const [0,14]}
       CTRL_R_mouth_lipU.X {type=MovieSceneScriptingFloatChannel, default=0, thinned_from=14, keys [[0,0],[470,0]]}
       CTRL_R_mouth_lipU.Y {type=MovieSceneScriptingFloatChannel, default=0, thinned_from=14, keys [[0,0],[259,0],[296,0.2369],[366,0.1454],[402,0.0012],[470,0]]}
       mouthPress_3 {type=MovieSceneScriptingFloatChannel, default=0, const [0,14]}
       lipsTogether_3 {type=MovieSceneScriptingFloatChannel, default=0, const [0,14]}
       CTRL_R_mouth_lipD.X {type=MovieSceneScriptingFloatChannel, default=0, thinned_from=14, keys [[0,0],[470,0]]}
       CTRL_R_mouth_lipD.Y {
        type=MovieSceneScriptingFloatChannel
        default=0
        thinned_from=14
        keys [[0,0],[259,0],[296,-0.3862],[366,-0.2371],[402,-0.0019],[428,-0.0009],[465,0],[470,0]]
       }
       mouthPress_2_2 {
        type=MovieSceneScriptingFloatChannel
        default=0
        thinned_from=14
        keys [[0,0],[259,0],[296,0.1502],[366,0.1502],[402,0.0026],[428,0.0013],[465,0],[470,0]]
       }
       lipsTogether_2_2 {
        type=MovieSceneScriptingFloatChannel
        default=0
        thinned_from=14
        keys [[0,0],[259,0],[296,0.2349],[366,0.2349],[402,0.0041],[428,0.0021],[465,0],[470,0]]
       }
       CTRL_R_mouth_smileDepress.X {type=MovieSceneScriptingFloatChannel, default=0, thinned_from=14, keys [[0,0],[470,0]]}
       CTRL_R_mouth_smileDepress.Y {
        type=MovieSceneScriptingFloatChannel
        default=0.3884
        thinned_from=14
        keys [[0,0.177],[56,0.3884],[77,0.3884],[113,0.3407],[259,0.3407],[296,0.3483],[366,0.3582],[402,0.1767],[470,0.177]]
       }
       CTRL_R_mouth_corner.X {
        type=MovieSceneScriptingFloatChannel
        default=0
        thinned_from=14
        keys [[0,0],[259,0],[296,1],[366,0.614],[402,0.0049],[428,0.0025],[465,0],[470,0]]
       }
       CTRL_R_mouth_corner.Y {
        type=MovieSceneScriptingFloatChannel
        default=0.4017
        thinned_from=14
        keys [[0,0.4017],[77,0.4017],[113,0.3524],[259,0.3524],[296,0.1879],[366,0.1154],[402,0.3956],[428,0.3986],[470,0.4017]]
       }
       lipsBlow_2 {type=MovieSceneScriptingFloatChannel, default=0, const [0,14]}
       mouthDimple_2 {type=MovieSceneScriptingFloatChannel, default=0, const [0,14]}
       mouthStretch_2 {type=MovieSceneScriptingFloatChannel, default=0, const [0,14]}
       mouthPurseU_2 {
        type=MovieSceneScriptingFloatChannel
        default=0
        thinned_from=14
        keys [[0,0],[259,0],[296,0.4888],[366,0.4888],[402,0.0085],[428,0.0043],[465,0],[470,0]]
       }
       mouthPurseD_2 {type=MovieSceneScriptingFloatChannel, default=0, const [0,14]}
       mouthFunnelU_2 {type=MovieSceneScriptingFloatChannel, default=0, thinned_from=14, keys [[0,0],[77,0],[113,0.0139],[259,0.0139],[296,0],[470,0]]}
       mouthFunnelD_2 {type=MovieSceneScriptingFloatChannel, default=0, thinned_from=14, keys [[0,0],[77,0],[113,0.0139],[259,0.0139],[296,0],[470,0]]}
       mouthCornerSharpnessU_2 {type=MovieSceneScriptingFloatChannel, default=0, const [0,14]}
       mouthCornerSharpnessD_2 {type=MovieSceneScriptingFloatChannel, default=0, const [0,14]}
       CTRL_L_mouth_suckBlow {type=MovieSceneScriptingFloatChannel, default=0, thinned_from=14, keys [[0,-1],[56,0],[366,0],[402,-0.9825],[428,-0.9912],[465,-1],[470,-1]]}
       CTRL_C_jaw.Location.X {type=MovieSceneScriptingFloatChannel, default=0, thinned_from=14, keys [[0,0],[470,0]]}
       CTRL_C_jaw.Location.Y {type=MovieSceneScriptingFloatChannel, default=0, thinned_from=14, keys [[0,0],[259,0],[296,-0.145],[366,-0.0891],[402,-0.0007],[470,0]]}
       CTRL_C_jaw.Location.Z {type=MovieSceneScriptingFloatChannel, default=0, thinned_from=14, keys [[0,0],[259,0],[296,0.1324],[366,0.0813],[402,0.0006],[470,0]]}
       jawOpenExtreme {type=MovieSceneScriptingFloatChannel, default=0, thinned_from=14, keys [[0,0],[259,0],[296,0.0151],[366,0.0151],[402,0.0003],[470,0]]}
       chinRaiseUL {type=MovieSceneScriptingFloatChannel, default=0, const [0,14]}
       chinRaiseUR {type=MovieSceneScriptingFloatChannel, default=0, const [0,14]}
       chinRaiseDL {type=MovieSceneScriptingFloatChannel, default=0, const [0,14]}
       chinRaiseDR {type=MovieSceneScriptingFloatChannel, default=0, const [0,14]}
       CTRL_R_brow.X {type=MovieSceneScriptingFloatChannel, default=0, thinned_from=14, keys [[0,0],[470,0]]}
       CTRL_R_brow.Y {
        type=MovieSceneScriptingFloatChannel
        default=-0.7525
        thinned_from=14
        keys [[0,-0.1464],[56,-0.0892],[77,-0.5178],[113,-0.2322],[136,-0.2322],[177,-0.4338],[259,-0.4338],[296,-0.7525],[366,-0.3409],[402,-0.3409],[428,-0.7525],[465,-0.1464],[470,-0.1464]]
       }
       CTRL_R_brow_inner {
        type=MovieSceneScriptingFloatChannel
        default=-0.0369
        thinned_from=14
        keys [[0,-0.0528],[56,0.4731],[77,0.2731],[113,0.2353],[136,0.2353],[177,0.5268],[259,0.5268],[296,0.4099],[366,-0.0369],[428,-0.0369],[465,-0.0528],[470,-0.0528]]
       }
       CTRL_R_brow_outer {type=MovieSceneScriptingFloatChannel, default=0, const [0,14]}
       CTRL_R_eye_blink {type=MovieSceneScriptingFloatChannel, default=0, const [0,14]}
       CTRL_R_eye_eyelidU.X {type=MovieSceneScriptingFloatChannel, default=0, thinned_from=14, keys [[0,0],[77,0],[113,-0.3385],[136,-0.3385],[177,0],[470,0]]}
       CTRL_R_eye_eyelidU.Y {
        type=MovieSceneScriptingFloatChannel
        default=0
        thinned_from=14
        keys [[0,0],[77,0],[113,-0.0458],[136,-0.0458],[177,0],[402,0],[428,0.2171],[465,0],[470,0]]
       }
       CTRL_R_eye_eyelidD.X {type=MovieSceneScriptingFloatChannel, default=0, thinned_from=14, keys [[0,0],[470,0]]}
       CTRL_R_eye_eyelidD.Y {
        type=MovieSceneScriptingFloatChannel
        default=0
        thinned_from=14
        keys [[0,0],[259,0],[296,0.4264],[366,0.2901],[402,0.2901],[428,0.1415],[465,0],[470,0]]
       }
       CTRL_R_eye_squintInner {type=MovieSceneScriptingFloatChannel, default=0, const [0,14]}
       CTRL_R_eye_cheekRaise {
        type=MovieSceneScriptingFloatChannel
        default=0.5905
        thinned_from=14
        keys [[0,0.5905],[56,0.5905],[77,0],[113,0.5905],[259,0.5905],[296,0],[366,0],[402,0.5801],[428,0.5853],[465,0.5905],[470,0.5905]]
       }
       CTRL_R_mouth_suckBlow {type=MovieSceneScriptingFloatChannel, default=0, thinned_from=14, keys [[0,-1],[56,0],[366,0],[402,-0.9825],[428,-0.9912],[465,-1],[470,-1]]}
       CTRL_C_eye.X {type=MovieSceneScriptingFloatChannel, default=0.115, const [0.115,14]}
       CTRL_C_eye.Y {type=MovieSceneScriptingFloatChannel, default=0, const [0,14]}
       CTRL_L_eye.X {type=MovieSceneScriptingFloatChannel, default=0, const [0,14]}
       CTRL_L_eye.Y {type=MovieSceneScriptingFloatChannel, default=0, const [0,14]}
       CTRL_R_eye.X {type=MovieSceneScriptingFloatChannel, default=0, const [0,14]}
       CTRL_R_eye.Y {type=MovieSceneScriptingFloatChannel, default=0, const [0,14]}
       eyesConvergence {type=MovieSceneScriptingFloatChannel, default=0, const [0,14]}
       eyelidsReduce {type=MovieSceneScriptingFloatChannel, default=0, const [0,14]}
       lookAtSwitch {type=MovieSceneScriptingFloatChannel, default=1, const [1,14]}
       CTRL_C_tongue.Location.X {type=MovieSceneScriptingFloatChannel, default=0, const [0,14]}
       CTRL_C_tongue.Location.Y {type=MovieSceneScriptingFloatChannel, default=0, const [0,14]}
       CTRL_C_tongue.Location.Z {type=MovieSceneScriptingFloatChannel, default=0, const [0,14]}
       wideNarrow {type=MovieSceneScriptingFloatChannel, default=0, const [0,14]}
       rollUpDown {type=MovieSceneScriptingFloatChannel, default=0, const [0,14]}
       rollLeftRight {type=MovieSceneScriptingFloatChannel, default=0, const [0,14]}
       CTRL_C_teethU.Location.X {type=MovieSceneScriptingFloatChannel, default=0, const [0,14]}
       CTRL_C_teethU.Location.Y {type=MovieSceneScriptingFloatChannel, default=0, thinned_from=14, keys [[0,0],[259,0],[296,0.4819],[428,0.4819],[465,0],[470,0]]}
       CTRL_C_teethU.Location.Z {type=MovieSceneScriptingFloatChannel, default=0, thinned_from=14, keys [[0,0],[470,0]]}
       rotateFwdBack {type=MovieSceneScriptingFloatChannel, default=0, const [0,14]}
       CTRL_C_teethD.Location.X {type=MovieSceneScriptingFloatChannel, default=0, const [0,14]}
       CTRL_C_teethD.Location.Y {type=MovieSceneScriptingFloatChannel, default=0, thinned_from=14, keys [[0,0],[259,0],[296,0.1897],[428,0.1897],[465,0],[470,0]]}
       CTRL_C_teethD.Location.Z {type=MovieSceneScriptingFloatChannel, default=0, thinned_from=14, keys [[0,0],[470,0]]}
       rotateFwdBack_2 {type=MovieSceneScriptingFloatChannel, default=0, const [0,14]}
      ]
     }]
   }
   {
    type=MovieScene3DTransformTrack
    sections [{
      type=MovieScene3DTransformSection
      channels [
       Location.X {type=MovieSceneScriptingDoubleChannel, default=0}
       Location.Y {type=MovieSceneScriptingDoubleChannel, default=0}
       Location.Z {type=MovieSceneScriptingDoubleChannel, default=0}
       Rotation.X {type=MovieSceneScriptingDoubleChannel, default=0}
       Rotation.Y {type=MovieSceneScriptingDoubleChannel, default=0}
       Rotation.Z {type=MovieSceneScriptingDoubleChannel, default=0}
       Scale.X {type=MovieSceneScriptingDoubleChannel, default=1}
       Scale.Y {type=MovieSceneScriptingDoubleChannel, default=1}
       Scale.Z {type=MovieSceneScriptingDoubleChannel, default=1}
      ]
     }]
   }
  ]
 }
 Head {
  id=#6
  parent_id=#5
  possessed_class=SkeletalMeshComponentBudgeted
  tracks [{
    type=MovieSceneControlRigParameterTrack
    display_name=Fortnite_Face_BoneTweaker_CtrlRig
    sections [{type=MovieSceneControlRigParameterSection}]
   }]
 }
 SkeletalMeshComponent0 {id=#7, parent_id=#1, possessed_class=SkeletalMeshComponentBudgeted, tracks []}
 SkeletalMeshComponent0 {id=#8, parent_id=#4, possessed_class=SkeletalMeshComponentBudgeted, tracks []}
]
display_rate {denominator=1, numerator=30}
tick_resolution {denominator=1, numerator=24000}
tracks []
